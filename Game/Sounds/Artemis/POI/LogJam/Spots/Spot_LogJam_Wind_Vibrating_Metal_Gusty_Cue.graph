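# Looping spot ambience of gusty wind vibrating metal at the LogJam POI, with attenuation and a two-voice concurrency cap.
# A short random delay feeds a quality switch that plays the same chain in its first two slots and nothing in the others.
# A wind loop reused from Blimps_Crashsites gets random pitch, then nested envelopes add a 30-second pitch drift, 120-second volume swells and a brief fade-in, varying the single sample.
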SoundCue /Game/Sounds/Artemis/POI/LogJam/Spots/Spot_LogJam_Wind_Vibrating_Metal_Gusty_Cue
root=Delay_0
nodes [
 {id=Delay_0, children [QualityLevel_0], properties {delayMax=0.1, delayMin=0.05}}
 {
  id=Enveloper_0
  children [Modulator_1]
  properties {
   bLoop=true
   bLoopIndefinitely=true
   loopEnd=30
   pitchCurve {editorCurveData {keys [{interpMode=RCIM_Cubic, time=0}, <at 15.9468 0.9 RCIM_Cubic>, {interpMode=RCIM_Cubic, time=30}]}}
   volumeCurve {editorCurveData {keys [<at 0 1>]}}
  }
 }
 {
  id=Enveloper_29
  children [Enveloper_3]
  properties {
   pitchCurve {editorCurveData {keys [{time=0}]}}
   volumeCurve {editorCurveData {keys [{arriveTangent=3.4293, interpMode=RCIM_Cubic, leaveTangent=3.4293, tangentMode=RCTM_User, time=0, value=0}, <at 0.4 1 RCIM_Cubic>]}}
  }
 }
 {
  id=Enveloper_3
  children [Enveloper_0]
  properties {
   bLoop=true
   bLoopIndefinitely=true
   loopEnd=120
   pitchCurve {editorCurveData {keys [{time=0}]}}
   volumeCurve {editorCurveData {keys [
      {arriveTangent=-0.01, interpMode=RCIM_Cubic, time=0, value=1}
      <at 10 0.6 RCIM_Cubic>
      {arriveTangent=-0.01, interpMode=RCIM_Cubic, leaveTangent=-0.01, time=20, value=1}
      <at 40 0.3 RCIM_Cubic>
      {arriveTangent=0.002, interpMode=RCIM_Cubic, leaveTangent=0.002, time=60, value=1}
      <at 90 0.4 RCIM_Cubic>
      <at 120 1>
     ]}}
  }
 }
 {id=Modulator_1, children [WavePlayer_1], properties {pitchMax=1, pitchMin=0.8}}
 {id=QualityLevel_0, children [Enveloper_29,Enveloper_29,null,null]}
 {
  id=WavePlayer_1
  properties {bLooping=true, soundWaveAssetPtr=/Game/Sounds/Artemis/POI/Blimps_Crashsites/Spots/Spot_Wind_Excited_Metal_Vibrating_Gusty_Loop}
 }
]
properties {
 attenuationSettings=&/Spot_Mountain_Window_Wind_Sphere_Att
 bEnableBusSends=false
 bEnableSubmixSends=false
 bOverrideConcurrency=true
 maxDistance=950
 soundClassObject=&Fort_Gadgets
 volumeMultiplier=1.2
 concurrencyOverrides {maxCount=2, voiceStealReleaseTime=2}
}
